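# Additive, unlit, two-sided material for the airglow of neon-tube skydive trails, enabled for beam trails and Niagara ribbons.
# Time plus the Twinkle_offset dynamic parameter scrolls curve-atlas rows for tube color and faster twinkle, multiplied into MF_FX_EmissiveAdjustments.
# A mask combines a Glow_ramp curve along V with a soft U-based profile, then applies camera depth fade.
# The other three dynamic parameters (Amplitude, Frequncy, Time_offset) are unused.
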Material /BRCosmetics/Athena/Effects/Materials/Skydiving/Neon_tubes/M_Skydive_Neon_Airglow
graph {
 properties {
  bUsedWithBeamTrails=true
  bUsedWithNiagaraRibbons=true
  blendMode=BLEND_Additive
  pixelDepthOffsetMode=PDOM_Legacy
  shadingModel=MSM_Unlit
  twoSided=true
  usedShadingModels=Unlit
 }
 nodes [
  n1 = Time()
  n2 = DynamicParameter() {paramNames [Amplitude, Frequncy, Time_offset, Twinkle_offset]}
  n3 = Add(A=@n1, B=Add(A=@n1, B=@n2.Twinkle_offset))
  n4 = CurveAtlasRowParameter(CurveTime=Frac(None=Multiply(A=@n3) {constB=0.1})) {
   atlas=&AthenaCurveAtlas01
   curve=/Game/Athena/LinearColorCurves/Curve_Neon_Ramp1
   defaultValue=63
   parameterName=Tube_color
  }
  n5 = Time()
  n6 = Add(A=@n5, B=Add(A=@n5, B=@n2.Twinkle_offset))
  n7 = CurveAtlasRowParameter(CurveTime=Frac(None=Multiply(A=@n6) {constB=0.3})) {
   atlas=&AthenaCurveAtlas01
   curve=/Game/Athena/LinearColorCurves/Curve_Neon_Twinkle_Ramp
   defaultValue=66
   parameterName=Twinkle
  }
  MaterialFunctionCall(Emissive=Multiply(A=Multiply(A=@n4) {constB=10}, B=@n7.R)) {materialFunction=&MF_FX_EmissiveAdjustments}
  n8 = TextureCoordinate()
  n9 = CurveAtlasRowParameter(CurveTime=ComponentMask(None=@n8) {g=true}) {
   atlas=&AthenaCurveAtlas01
   curve=/Game/Athena/LinearColorCurves/Curve_Neon_Trail_Ramp
   defaultValue=65
   parameterName=Glow_ramp
  }
  n10 = ComponentMask(None=@n8) {r=true}
  n11 = Multiply(A=Clamp(None=LinearInterpolate(Alpha=@n10) {constB=5}), B=OneMinus(None=@n10))
  n12 = Subtract(A=Multiply(A=@n11) {constB=2}, B=Power(Base=@n11))
  Multiply(
   A=Power(Base=Multiply(A=@n9.R, B=@n12))
   B=MaterialFunctionCall(Fade Length=Constant() {r=50}) {materialFunction=&CameraDepthFade}.Result
  )
 ]
}
