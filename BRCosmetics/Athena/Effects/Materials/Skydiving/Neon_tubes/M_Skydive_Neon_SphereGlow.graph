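# Additive, unlit sprite material for Niagara and particle systems that draws a soft, twinkling neon glow in the Neon Tubes skydiving effect.
# Emissive color cycles through a curve atlas color ramp, scaled by a faster twinkle ramp, with the Twinkle_offset dynamic parameter shifting both.
# A procedural mask combines sharpened UV edge falloff with an exponential radial gradient and particle alpha.
# Depth fade softens intersections on the default shading path and is skipped on mobile.
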Material /BRCosmetics/Athena/Effects/Materials/Skydiving/Neon_tubes/M_Skydive_Neon_SphereGlow
graph {
 properties {
  bUsedWithNiagaraSprites=true
  bUsedWithParticleSprites=true
  blendMode=BLEND_Additive
  pixelDepthOffsetMode=PDOM_Legacy
  shadingModel=MSM_Unlit
  twoSided=true
  usedShadingModels=Unlit
 }
 nodes [
  n1 = Time()
  n2 = DynamicParameter() {paramNames [Amplitude, Frequncy, Time_offset, Twinkle_offset]}
  n3 = Add(A=@n1, B=Add(A=@n1, B=@n2.Twinkle_offset))
  n4 = CurveAtlasRowParameter(CurveTime=Frac(None=Multiply(A=@n3) {constB=0.1})) {
   atlas=&AthenaCurveAtlas01
   curve=/Game/Athena/LinearColorCurves/Curve_Neon_Ramp1
   defaultValue=63
   parameterName=Tube_color
  }
  n5 = Time()
  n6 = Add(A=@n5, B=Add(A=@n5, B=@n2.Twinkle_offset))
  n7 = CurveAtlasRowParameter(CurveTime=Frac(None=Multiply(A=@n6) {constB=0.3})) {
   atlas=&AthenaCurveAtlas01
   curve=/Game/Athena/LinearColorCurves/Curve_Neon_Twinkle_Ramp
   defaultValue=66
   parameterName=Twinkle
  }
  MaterialFunctionCall(Emissive=Multiply(A=Multiply(A=@n4) {constB=10}, B=@n7.R)) {materialFunction=&MF_FX_EmissiveAdjustments}
  n8 = DepthFade() {fadeDistanceDefault=20}
  n9 = TextureCoordinate()
  n10 = ComponentMask(None=@n9) {g=true}
  n11 = OneMinus(None=@n10)
  n12 = ComponentMask(None=@n9) {r=true}
  n13 = OneMinus(None=@n12)
  n14 = Multiply(A=LinearInterpolate(Alpha=Multiply(A=@n10, B=@n11)) {constB=4}, B=LinearInterpolate(Alpha=Multiply(A=@n12, B=@n13)) {constB=4})
  n15 = Multiply(A=@n14, B=@n14)
  n16 = Multiply(
   A=Multiply(A=@n15, B=@n15)
   B=MaterialFunctionCall() {materialFunction=&RadialGradientExponential}.RadialGradientExponential
  )
  Multiply(A=ParticleColor().A, B=ShadingPathSwitch(Default=Multiply(A=@n8, B=@n16), Mobile=@n16) {desc=Migrated from FeatureLevelSwitch})
 ]
}
