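# Glowing two-sided neon tube material for mesh particles in a skydiving cosmetic effect.
# Emissive combines a view-angle falloff, time-cycled color ramps chosen by a main/sub tube static switch, and a twinkle ramp.
# Vertex offset waves the tube in sine/cosine motion from panned UVs and world height, driven by particle dynamic parameters.
# Non-mobile paths lift tubes near the actor under top-down views and use a precise custom node against depth divergence.
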Material /BRCosmetics/Athena/Effects/Materials/Skydiving/Neon_tubes/M_Skydive_Neon_light1
graph {
 properties {
  bUsedWithMeshParticles=true
  bUsedWithNiagaraMeshParticles=true
  pixelDepthOffsetMode=PDOM_Legacy
  twoSided=true
  usedShadingModels=Default Lit
 }
 nodes [
  n1 = DotProduct(A=CameraVectorWS(), B=VertexNormalWS())
  n2 = LinearInterpolate(Alpha=OneMinus(None=Abs(None=@n1))) {constA=1, constB=0.4}
  n3 = Multiply(A=@n2, B=@n2)
  n4 = LinearInterpolate(Alpha=Power(Base=@n3) {constExponent=5}) {constA=0.05}
  n5 = Time()
  n6 = DynamicParameter() {paramNames [BP_floatSpeed, Frequncy, Time_offset, Twinkle_offset]}
  n7 = Add(A=@n5, B=@n6.Twinkle_offset)
  n8 = Frac(None=Multiply(A=Add(A=@n5, B=@n7)) {constB=0.1})
  n9 = CurveAtlasRowParameter(CurveTime=@n8) {
   atlas=&AthenaCurveAtlas01
   curve=/Game/Athena/LinearColorCurves/Curve_Neon_Ramp1
   defaultValue=63
   parameterName=Tube_color
  }
  n10 = CurveAtlasRowParameter(CurveTime=@n8) {
   atlas=&AthenaCurveAtlas01
   curve=/Game/Athena/LinearColorCurves/Curve_Neon_Ramp2
   defaultValue=64
   parameterName=sub_Tube_color
  }
  n11 = Multiply(A=StaticSwitch(True=@n9, False=@n10, Value=StaticBoolParameter() {defaultValue=true, parameterName=Mani_tube})) {constB=8}
  n12 = Time()
  n13 = Add(A=@n12, B=Add(A=@n12, B=@n6.Twinkle_offset))
  n14 = CurveAtlasRowParameter(CurveTime=Frac(None=Multiply(A=@n13) {constB=0.3})) {
   atlas=&AthenaCurveAtlas01
   curve=/Game/Athena/LinearColorCurves/Curve_Neon_Twinkle_Ramp
   defaultValue=66
   parameterName=Twinkle
  }
  MaterialFunctionCall(Emissive=Multiply(A=Multiply(A=@n4, B=@n11), B=@n14.R)) {materialFunction=&MF_FX_EmissiveAdjustments}
  ParticlePositionWS()
  Power(Base=OneMinus(None=@n3)) {constExponent=5}
  n15 = TextureCoordinate()
  n16 = OneMinus(None=ComponentMask(None=@n15) {g=true})
  n17 = Multiply(A=Time(), B=@n6.BP_floatSpeed) {constB=4.2}
  n18 = ComponentMask(None=WorldPosition().XYZ) {b=true}
  n19 = Divide(A=Constant() {r=1618}, B=Multiply(A=@n6.Frequncy) {constB=-1})
  n20 = Panner(Coordinate=@n15, Time=Add(A=Add(A=@n6.Time_offset, B=@n17), B=Divide(A=@n18, B=@n19) {constB=1618})) {speedY=0.5}
  n21 = Multiply(A=Frac(None=ComponentMask(None=@n20) {g=true}))
  n22 = Sine(None=@n21)
  n23 = Cosine(None=@n21)
  n24 = Constant()
  n25 = Multiply(
   A=Sine(None=Divide(A=@n16) {constB=4})
   B=Multiply(
    A=MaterialFunctionCall(R=@n22, G=@n23, B=@n24) {materialFunction=&AppendMany}.RGB
    B=ScalarParameter() {defaultValue=20, parameterName=Amplitude}
   ) {constB=15}
  )
  Reroute(None=@n25)
  n26 = Subtract(A=ActorPositionWS(), B=CameraPositionWS())
  n27 = Subtract(A=ObjectPositionWS(), B=ActorPositionWS())
  n28 = Add(A=Constant3Vector() {constant=<rgba 0 0 10 0>}, B=@n25)
  n29 = If(
   A=Abs(None=MaterialFunctionCall(Vector 3=@n27) {materialFunction=/Engine/Functions/Engine_MaterialFunctions02/Utility/VectorLength}.V3 Length)
   B=Constant() {r=65}
   "A > B"=@n25
   "A == B"=@n28
   "A < B"=@n28
  )
  MaterialExpressionIf_1 = If(
   A=DotProduct(A=Normalize(VectorInput=@n26), B=Constant3Vector() {constant=<rgba 0 0 1 0>})
   B=Constant() {r=-0.9}
   "A > B"=@n25
   "A == B"=@n29
   "A < B"=@n29
  )
  n30 = Custom!(None=@MaterialExpressionIf_1) {
   code="##if SM5_PROFILE || SM4_PROFILE\r\n\tprecise float3 PreciseInput = Input;\r\n\treturn PreciseInput;\r\n##else\r\n\treturn Input;\r\n##endif\r\n"
   description=Fix Depth Divergence
   inputs [{inputName=Input, input {expression=&self:MaterialExpressionIf_1}}]
  }
  ShadingPathSwitch(Default=@n30, Mobile=@n25) {desc=Migrated from FeatureLevelSwitch}
 ]
}
